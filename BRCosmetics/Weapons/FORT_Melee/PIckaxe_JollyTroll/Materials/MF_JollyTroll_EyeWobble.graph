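# Material function for the JollyTroll pickaxe that, going by its name, adds wobbling eyes to base color.
# A time-driven sine, phased by V and remapped, scales two drift parameters into UV offsets, with the horizontal one also weighted by U.
# The FX texture's green channel, sampled at those UVs, serves as both color and blend alpha over incoming BaseColor.
# YDriftDistance drives the U offset and XDriftDistance the V offset, opposite to their names.
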MaterialFunction /BRCosmetics/Weapons/FORT_Melee/PIckaxe_JollyTroll/Materials/MF_JollyTroll_EyeWobble
graph {nodes [
  n1 = GetMaterialAttributes(None=FunctionInput() {inputType=FunctionInput_MaterialAttributes}) {attributeGetTypes [#1]}
  n2 = MaterialFunctionCall(Float2=TextureCoordinate()) {materialFunction=&BreakOutFloat2Components}
  n3 = TextureCoordinate()
  n4 = ComponentMask(None=@n3) {g=true}
  n5 = Multiply(A=Time(), B=ScalarParameter() {defaultValue=30, group=JollyTroll, parameterName=Wobble Speed})
  n6 = ConstantBiasScale(None=Sine(None=Add(A=@n4, B=@n5)))
  n7 = Multiply(
   A=ConstantBiasScale(None=ComponentMask(None=@n3) {r=true}) {bias=0.15, scale=0.25}
   B=Multiply(A=@n6, B=ScalarParameter() {defaultValue=0.01, group=JollyTroll, parameterName=YDriftDistance}) {constB=0.5}
  )
  n8 = Multiply(A=@n6, B=ScalarParameter() {defaultValue=-0.008, group=JollyTroll, parameterName=XDriftDistance}) {constB=0.015}
  n9 = TextureSample(UVs=MaterialFunctionCall(X=Subtract(A=@n2.R, B=@n7), Y=Add(A=@n2.G, B=@n8)) {materialFunction=&MakeFloat2}.Result) {texture=/BRCosmetics/Weapons/FORT_Melee/PIckaxe_JollyTroll/Textures/T_Pickaxe_JollyTroll_FX}
  FunctionOutput(None=SetMaterialAttributes!(MaterialAttributes=@n1.MaterialAttributes, BaseColor=LinearInterpolate(A=@n1.BaseColor, B=@n9.G, Alpha=@n9.G)) {attributeSetTypes [#1]})
 ]}
