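# Physics asset for the mouth of the JollyTroll pickaxe, a two-handed melee weapon cosmetic.
# It has two capsule bodies, one on the Melee_TwoHanded_Handle bone and a smaller one on the dyn_Mounth01 bone.
# A single constraint joins the mouth bone to the handle, with swing and twist limit angles all set to zero.
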PhysicsAsset /BRCosmetics/Weapons/FORT_Melee/PIckaxe_JollyTroll/Meshes/JollyTroll_AxePhysics_Mouth
bodies {
 Melee_TwoHanded_Handle [{
   shapeType=Capsule
   length=2.6246
   radius=1.3116
   shapeName=Melee_TwoHanded_Handle_capsule
   center=<2.3666 -1.5321 24.1485>
   rotation=<rot 0.1552 -90 -2.2017>
  }]
 dyn_Mounth01 [{shapeType=Capsule, length=1.6188, radius=0.75, shapeName=dyn_Mounth01_capsule, center=<2 0 0>, rotation=<rot 90 0 0>}]
}
constraints [{bone1Name=dyn_Mounth01, bone2Name=Melee_TwoHanded_Handle, swing1LimitDegrees=0, swing2LimitDegrees=0, twistLimitDegrees=0}]
