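# Physics asset for the JollyTroll pickaxe, giving its hat, moustache and decoration bones physical bodies for secondary motion.
# Each bone has a single capsule, and constraint chains root at the Melee_TwoHanded_Handle body: a two-bone hat, three-bone moustache sides and two-bone decorations on each side.
# Tight swing and twist limits keep motion subtle, mostly loosening toward chain tips except on the hat, whose base swings widest on one axis.
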PhysicsAsset /BRCosmetics/Weapons/FORT_Melee/PIckaxe_JollyTroll/Meshes/JollyTroll_Axe_Physics
bodies {
 Melee_TwoHanded_Handle [{
   shapeType=Capsule
   length=2.6246
   radius=1.3116
   shapeName=Melee_TwoHanded_Handle_capsule
   center=<2.3666 -1.5321 24.1485>
   rotation=<rot 0.1552 -90 -2.2017>
  }]
 dyn_Hat01 [{shapeType=Capsule, length=1.75, radius=1.7619, shapeName=dyn_Hat01_capsule, center=<2 0 0>, rotation=<rot 90 0 0>}]
 dyn_Hat02 [{shapeType=Capsule, length=2.5587, radius=1.2668, shapeName=dyn_Hat02_capsule, center=<2 0 0>, rotation=<rot 90 0 0>}]
 dyn_L_Decoration01 [{shapeType=Capsule, length=2, radius=1, shapeName=dyn_L_Decoration01_capsule, center=<2 0 0>, rotation=<rot 90 0 0>}]
 dyn_L_Decoration02 [{shapeType=Capsule, length=2, radius=1, shapeName=dyn_L_Decoration02_capsule, center=<2.5 0 0>, rotation=<rot 90 0 0>}]
 dyn_L_Moustache01 [{shapeType=Capsule, length=2.25, radius=1.25, shapeName=dyn_L_Moustache01_capsule, center=<2.5 0 0>, rotation=<rot 90 0 0>}]
 dyn_L_Moustache02 [{shapeType=Capsule, length=2.8881, radius=0.75, shapeName=dyn_L_Moustache02_capsule, center=<3 0 0>, rotation=<rot 90 0 0>}]
 dyn_L_Moustache03 [{shapeType=Capsule, length=1.5, radius=0.75, shapeName=dyn_L_Moustache03_capsule, center=<1.6959 0 0>, rotation=<rot 90 0 0>}]
 dyn_R_Decoration01 [{shapeType=Capsule, length=2, radius=1, shapeName=dyn_R_Decoration01_capsule, center=<2 0 0>, rotation=<rot 90 0 0>}]
 dyn_R_Decoration02 [{shapeType=Capsule, length=2, radius=1, shapeName=dyn_R_Decoration02_capsule, center=<2.5 0 0>, rotation=<rot 90 0 0>}]
 dyn_R_Moustache01 [{shapeType=Capsule, length=2.25, radius=1.25, shapeName=dyn_R_Moustache01_capsule, center=<2.5 0 0>, rotation=<rot 90 0 0>}]
 dyn_R_Moustache02 [{shapeType=Capsule, length=2.8881, radius=0.75, shapeName=dyn_R_Moustache02_capsule, center=<3 0 0>, rotation=<rot 90 0 0>}]
 dyn_R_Moustache03 [{shapeType=Capsule, length=1.5, radius=0.75, shapeName=dyn_R_Moustache03_capsule, center=<1.6959 0 0>, rotation=<rot 90 0 0>}]
}
constraints [
 {bone1Name=dyn_L_Moustache01, bone2Name=Melee_TwoHanded_Handle, swing1LimitDegrees=0, swing2LimitDegrees=2, twistLimitDegrees=0}
 {bone1Name=dyn_L_Moustache02, bone2Name=dyn_L_Moustache01, swing1LimitDegrees=5, swing2LimitDegrees=7, twistLimitDegrees=1}
 {bone1Name=dyn_L_Moustache03, bone2Name=dyn_L_Moustache02, swing1LimitDegrees=5, swing2LimitDegrees=10, twistLimitDegrees=1}
 {bone1Name=dyn_R_Moustache01, bone2Name=Melee_TwoHanded_Handle, swing1LimitDegrees=2, swing2LimitDegrees=0, twistLimitDegrees=0}
 {bone1Name=dyn_R_Moustache02, bone2Name=dyn_R_Moustache01, swing1LimitDegrees=7, swing2LimitDegrees=5, twistLimitDegrees=1}
 {bone1Name=dyn_R_Moustache03, bone2Name=dyn_R_Moustache02, swing1LimitDegrees=10, swing2LimitDegrees=5, twistLimitDegrees=1}
 {bone1Name=dyn_R_Decoration01, bone2Name=Melee_TwoHanded_Handle, swing1LimitDegrees=5, swing2LimitDegrees=3, twistLimitDegrees=1}
 {bone1Name=dyn_R_Decoration02, bone2Name=dyn_R_Decoration01, swing1LimitDegrees=10, swing2LimitDegrees=10, twistLimitDegrees=1}
 {bone1Name=dyn_L_Decoration01, bone2Name=Melee_TwoHanded_Handle, swing1LimitDegrees=5, swing2LimitDegrees=3, twistLimitDegrees=1}
 {bone1Name=dyn_L_Decoration02, bone2Name=dyn_L_Decoration01, swing1LimitDegrees=10, swing2LimitDegrees=10, twistLimitDegrees=1}
 {bone1Name=dyn_Hat02, bone2Name=dyn_Hat01, swing1LimitDegrees=1, swing2LimitDegrees=5, twistLimitDegrees=1}
 {bone1Name=dyn_Hat01, bone2Name=Melee_TwoHanded_Handle, swing1LimitDegrees=0, swing2LimitDegrees=12.5, twistLimitDegrees=0}
]
