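# Animation Blueprint for the Jolly Troll pickaxe that adds physics-driven secondary motion to the weapon mesh.
# It copies the pose and curves from the attached parent, then chains two RigidBody nodes, a main physics asset and a mouth-specific one.
# Each update computes custom gravity for the first node via a deprecated function and lowers both nodes' alpha during knee slides.
# It also sets two "Hide Element" material parameters every update.
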AnimBlueprint /BRCosmetics/Weapons/FORT_Melee/PIckaxe_JollyTroll/Meshes/JollyTroll_Axe_AnimBP
target_skeleton=&/JollyTroll_Axe_Skeleton
graphs [
 AnimGraph {nodes [
   n1 = Get Alpha() -> (Alpha:float)
   n2 = RigidBody!(
    ComponentPose=Local To Component(LocalPose=Copy Pose From Mesh() {bCopyCurves=true, bUseAttachedParent=true})
    Alpha=@n1
    OverrideWorldGravity:Vector=Get GravityOverride()
    ExternalForce
   ) -> (Pose) {
    bOverrideWorldGravity=true
    overridePhysicsAsset=&/JollyTroll_Axe_Physics
    simulationSpace=BaseBoneSpace
    alphaBoolBlend {blendOption=Linear}
    baseBoneRef {boneName=Melee_TwoHanded_Handle}
    componentAppliedLinearAccClamp=<5000 5000 5000>
    componentLinearAccScale=<0.4 0.4 0.45>
    componentLinearVelScale=<0.4 0.4 0.45>
    simSpaceSettings {
     maxAngularAcceleration=1000
     maxAngularVelocity=3
     maxLinearAcceleration=700
     maxLinearVelocity=500
     velocityScaleZ=0
     externalLinearDragV=<0.2 0.2>
    }
   }
   n3 = RigidBody(ComponentPose=@n2, Alpha=@n1, OverrideWorldGravity="0,0,-750") {
    bOverrideWorldGravity=true
    overridePhysicsAsset=&/JollyTroll_AxePhysics_Mouth
    alphaBoolBlend {blendOption=Linear}
    baseBoneRef {boneName=b_AR_Root}
    componentAppliedLinearAccClamp=<3000 3000 3000>
    componentLinearAccScale=<0.5 0.5 0.5>
    componentLinearVelScale=<0.5 0.5 0.5>
    simSpaceSettings {
     maxAngularAcceleration=1000
     maxAngularVelocity=3
     maxLinearAcceleration=700
     maxLinearVelocity=500
     velocityScaleZ=0
     externalLinearDragV=<0.2 0.2>
    }
   }
   Output Pose(Result=Component To Local(ComponentPose=@n3)) {layerGroup=None}
   AnimGraphNode_SubInput_0 = Input Pose()
  ]}
 EventGraph {nodes [
   Event BlueprintUpdateAnimation()
   > n1 = Sequence!(execute) -> (then_0, then_1, then_2:Exec, then_3:Exec)
   > n2 = Gravity Override State Function DEPRECATED(CharacterPartAnimInstance=Self-Reference(), GameTimeInSeconds=GetGameTimeInSeconds(), GravityOverrideParams=Get Gravity Override Params())
   > Set GravityOverride(GravityOverride=@n2)
   n3 = Get Dynamic Material Instance(execute=@n1.then_1, SkeletalMeshComponent=GetPartSkeletalMeshComponent())
   > SetScalarParameterValue(self=@n3, ParameterName=Hide Element 03, Value=0)
   > SetScalarParameterValue(self=@n3, ParameterName=Hide Element 06, Value=1)
   n4 = Branch(execute=@n1.then_2, Condition=Get bIsKneeSliding())
   > Set Alpha(Alpha:float=Get Alpha_CrouchSprint() -> (Alpha_CrouchSprint:float)) -> (Output_Get:float)
   Set Alpha(execute=@n4.else, Alpha=Get Alpha_regular() -> (Alpha_regular:float))
  ]}
]
variables [
 Alpha {type=number, default=1}
 Alpha_CrouchSprint {type=number, default=0.2}
 Alpha_regular {type=number, default=1}
 Gravity Override Params {type=GravityOverrideParamsStruct}
 GravityOverride {type=Vector}
]
