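# Translucent unlit Niagara sprite material for twinkling, emissive confetti particles on the JollyTroll pickaxe.
# A sine of particle relative time, offset by particle random, is scaled and clamped to boost particle color, then passed through MF_FX_EmissiveAdjustments.
# The mobile shading path skips the twinkle and uses plain particle color.
# Particle alpha multiplies a selectable channel of a 4x4 confetti SubUV texture; time-of-day and TwinkleSpeed branches appear unused.
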Material /BRCosmetics/Weapons/FORT_Melee/PIckaxe_JollyTroll/VFX/M_JollyTroll_Confetti_ParticleColor_Emissive
graph {
 properties {
  bUsedWithNiagaraSprites=true
  blendMode=BLEND_Translucent
  pixelDepthOffsetMode=PDOM_Legacy
  shadingModel=MSM_Unlit
  translucencyLightingMode=TLM_VolumetricPerVertexDirectional
  usedShadingModels=Unlit
 }
 nodes [
  n1 = ParticleRandom()
  n2 = Add(A=Multiply(A=ParticleRelativeTime()) {constB=2}, B=@n1)
  n3 = Clamp(None=Multiply(A=Sine(None=@n2)) {constB=10}) {maxDefault=9, minDefault=1}
  n4 = ParticleColor()
  n5 = Add(A=Multiply(A=Multiply(A=@n3, B=@n4.RGB), B=Constant() {r=1}), B=@n4.RGB)
  MaterialFunctionCall(Emissive=ShadingPathSwitch(Default=@n5, Deferred=@n5, Mobile=@n4.RGB) {desc=Migrated from FeatureLevelSwitch}) {materialFunction=&MF_FX_EmissiveAdjustments}
  n6 = Constant4Vector() {constant=<rgba 1 2 1 1>}
  MaterialFunctionCall(Value_Morning=@n6.R, Value_Day=@n6.R, Value_Evening=@n6.R, Value_Night=@n6.R) {materialFunction=&Scalar_Time_of_Day}
  MaterialFunctionCall(Speed=Multiply(A=@n1, B=ScalarParameter() {defaultValue=4, parameterName=TwinkleSpeed})) {materialFunction=&DebugTimeSine}
  Multiply(
   A=@n4.A
   B=StaticComponentMaskParameter(None=TextureSample() {texture=/Game/Effects/Fort_Effects/Textures/Confetti/T_Confetti_SubUV_4X4_A}.B) {defaultB=true, parameterName=PickChannel}
  )
 ]
}
