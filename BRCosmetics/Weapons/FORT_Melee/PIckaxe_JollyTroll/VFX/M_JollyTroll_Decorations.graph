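# Translucent Niagara sprite material for decoration particles in the Jolly Troll pickaxe VFX.
# It samples the decorations texture through the Tex1 parameter and multiplies its RGB and alpha separately by particle color and alpha.
# This lets the particle system control the tint and fade of each sprite.
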Material /BRCosmetics/Weapons/FORT_Melee/PIckaxe_JollyTroll/VFX/M_JollyTroll_Decorations
graph {
 properties {bUsedWithNiagaraSprites=true, blendMode=BLEND_Translucent, pixelDepthOffsetMode=PDOM_Legacy, usedShadingModels=Default Lit}
 nodes [
  n1 = TextureSampleParameter2D() {parameterName=Tex1, texture=&/T_Decorations}
  n2 = ParticleColor()
  Multiply(A=@n1.RGB, B=@n2.RGB)
  Multiply(A=@n1.A, B=@n2.A)
 ]
}
