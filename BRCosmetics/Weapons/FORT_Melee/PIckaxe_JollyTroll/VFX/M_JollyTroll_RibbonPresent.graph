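# Translucent, Default Lit Niagara ribbon material for the Jolly Troll pickaxe VFX, drawing a scrolling present-themed texture with animated wobble.
# Panned texture R and G channels are tinted by two color parameters, added to B, then scaled by dimmed particle color; alpha multiplies particle alpha.
# Panning smooth-noise and normal-map samples, offset by particle random, build a three-axis world position offset that fades out as V approaches one.
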Material /BRCosmetics/Weapons/FORT_Melee/PIckaxe_JollyTroll/VFX/M_JollyTroll_RibbonPresent
graph {
 properties {
  bCastRayTracedShadows=false
  bUsedWithNiagaraRibbons=true
  blendMode=BLEND_Translucent
  pixelDepthOffsetMode=PDOM_Legacy
  usedShadingModels=Default Lit
 }
 nodes [
  n1 = TextureCoordinate()
  n2 = Add(
   A=Multiply(A=ComponentMask(None=@n1) {r=true}, B=ScalarParameter() {defaultValue=1, parameterName=ScaleU})
   B=ScalarParameter() {parameterName=Offset_U}
  )
  n3 = Add(A=ComponentMask(None=@n1) {g=true}, B=ScalarParameter() {parameterName=Offset_V})
  n4 = ScalarParameter() {defaultValue=-1.5, parameterName=SpeedU}
  n5 = ScalarParameter() {parameterName=SpeedV}
  n6 = TextureSampleParameter2D(UVs=Panner(Coordinate=AppendVector(A=@n2, B=@n3), Speed=AppendVector(A=@n4, B=@n5))) {parameterName=Tex2, texture=&/T_RibbonPresent_Color}
  n7 = Multiply(A=@n6.R, B=VectorParameter() {parameterName=Color1, defaultValue=<rgba 0.8229 0.024 0 0>}.RGB)
  n8 = Multiply(A=@n6.G, B=VectorParameter() {parameterName=Color2, defaultValue=<rgba 0.7448 0.427 0.149 0>}.RGB)
  n9 = ParticleColor()
  Multiply(A=Add(A=Add(A=@n7, B=@n8), B=@n6.B), B=Multiply(A=@n9.RGB) {constB=0.3})
  n10 = ParticleRandom()
  n11 = VectorParameter() {group=WPO, parameterName=WPO Wave UV, sortPriority=6, defaultValue=<rgba 0.3 0.3 0.1 0.4>}
  n12 = AppendVector(A=@n11.R, B=@n11.R)
  n13 = TextureCoordinate()
  n14 = Add(A=AppendVector(A=@n10, B=@n10), B=Panner(Coordinate=Multiply(A=@n13, B=@n12), Speed=AppendVector(A=@n11.B, B=@n11.B)))
  n15 = Multiply(
   A=LinearInterpolate(Alpha=TextureSample(UVs=@n14) {texture=&T_Smooth_tile}.R) {constA=-0.25}
   B=ScalarParameter() {defaultValue=1, group=WPO, parameterName=WPO Side Intensity}
  ) {constB=10}
  n16 = VectorParameter() {group=WPO, parameterName=WPO Side UV, sortPriority=6, defaultValue=<rgba 0.3 0.3 -0.1 0.3>}
  n17 = Multiply(A=@n13, B=AppendVector(A=@n16.R, B=@n16.R))
  n18 = AppendVector(A=@n16.B, B=@n16.B)
  n19 = TextureSample(UVs=Add(A=AppendVector(A=@n10, B=@n10), B=Panner(Coordinate=@n17, Speed=@n18))) {samplerType=SAMPLERTYPE_Normal, texture=/Game/Characters/Player/Female/Medium/Bodies/F_MED_Bistro_Astronaut/FX/T_GaussianNoise_Normal}
  n20 = Multiply(A=ComponentMask(None=@n19.RGB) {r=true}, B=ScalarParameter() {defaultValue=10, group=WPO, parameterName=WPO Horizontal Intensity}) {constB=10}
  n21 = Multiply(A=ComponentMask(None=@n19.RGB) {g=true}, B=ScalarParameter() {defaultValue=2, group=WPO, parameterName=WPO Vertical Intensity}) {constB=10}
  n22 = ComponentMask(None=TextureCoordinate()) {g=true}
  Multiply(
   A=LinearInterpolate(
    B=MaterialFunctionCall(R=@n15, G=@n20, B=@n21) {materialFunction=&AppendMany}.RGB
    Alpha=OneMinus(None=@n22)
   )
   B=ScalarParameter() {defaultValue=5, group=WPO, parameterName=WPO Intensity}
  ) {constB=10}
  Multiply(A=@n6.A, B=@n9.A)
 ]
}
